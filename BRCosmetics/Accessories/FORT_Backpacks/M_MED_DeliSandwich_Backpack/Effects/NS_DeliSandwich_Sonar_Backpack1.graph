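# Niagara system for a repeating sonar pulse effect on the DeliSandwich backpack cosmetic.
# DeliSandwish_VFX emits pink-red meshes that drift along X, grow and fade, with alpha also enveloped by normalized loop age so each train tapers.
# NE_Decal and NE_DecalCenter burst one X-facing sprite per loop, a red glow and smaller warm-white flash, faded through a LightAlpha curve.
# Loop delays and durations keep all three emitters on a matching period.
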NiagaraSystem /BRCosmetics/Accessories/FORT_Backpacks/M_MED_DeliSandwich_Backpack/Effects/NS_DeliSandwich_Sonar_Backpack1
dependencies {
 usedDataInterfaces [&NiagaraDataInterfaceColorCurve, &NiagaraDataInterfaceCurve]
 usedDynamicInputs []
 usedModules [&SystemState]
}
emitters {
 DeliSandwish_VFX {
  emitterUpdateScript [
   &EmitterState(Inactive Response=::NewEnumerator1, Life Cycle Mode=::NewEnumerator1, Loop Delay=0.9, Loop Duration=1.95, UseLoopDelay=true)
   &SpawnRate(SpawnRate=3, Use Spawn Probability=false)
  ]
  particleSpawnScript [
   /Niagara/Modules/Spawn/Initialization/InitializeParticle(
    Lifetime=0.95
    Mass=&UniformRangedFloat(Minimum=0.6, Maximum=1.2)
    Mesh Scale=&VectorFromFloat(Value=0.45)
    Module.Write Color=true
    Module.Write Lifetime=true
    Module.Write Mass=true
    Module.Write Material Random=true
    Module.Write Position=false
    Module.Write Scale=true
    Module.Write SpriteRotation=false
    Module.Write SpriteSize=false
    Position=&SimulationPosition()
    Sprite Rotation=&NormalizedAngleToDegrees(NormalizedAngle=&UniformRangedFloat(Minimum=0, Maximum=1))
   )
   &Color(Color=&ColorFromCurve(ColorCurve=DataInterface() {propertyValues {
       AlphaCurve {keys [<at 0.0013 1>]}
       BlueCurve {keys [<at 0 0.7034 RCIM_Cubic>]}
       GreenCurve {keys [<at 0 0.5241 RCIM_Cubic>]}
       RedCurve {keys [<at 0 3 RCIM_Cubic>]}
      }}))
   &AddVelocity(Velocity=<22 0 0>)
   &DynamicMaterialParameters(
    Index 0 Param 1=2
    Index 0 Param 4=8
    Module.Param0WriteEnabled=false
    Module.Param1WriteEnabled=false
    Module.Param2WriteEnabled=true
    Module.Param3WriteEnabled=true
   ) {module=DynamicMaterialParameters001}
   &AddRotationalVelocity(Rotation Rate=&UniformRangedVector(Minimum=<-0.1 0 0>, Maximum=<0.1 0 0>))
   &InitialMeshOrientation(
    Orientation Vector=&UniformRangedVector(Minimum=<-0.5 0 0>, Maximum=<0.5 0 0>)
    Use Orientation Vector=true
    Use Rotation Vector=false
   ) {module=InitialMeshOrientation001}
  ]
  particleUpdateScript [
   &Drag(Use Linear Drag=true, Use Rotational Drag=false)
   &SolveForcesAndVelocity(Clamp Velocity=false, Limit Acceleration=false)
   &SolveRotationalForcesAndVelocity()
   /SetVariables(Particles.LightAlpha:NiagaraFloat=&FloatFromCurve(
     FloatCurve=DataInterface() {propertyValues {Curve {keys [
         <at 0 0>
         <at 0.1 1>
         {arriveTangent=-3.6605, interpMode=RCIM_Cubic, leaveTangent=-3.6605, tangentMode=RCTM_User, time=0.45, value=1}
         {arriveTangent=-3.1112, interpMode=RCIM_Cubic, leaveTangent=-3.1112, tangentMode=RCTM_User, time=1.0049, value=-1}
        ]}}}
     CurveIndex=$Emitter.NormalizedLoopAge
    )) {isSetParameters=true}
   &ParticleState()
   &ScaleMeshSize(Scale Factor=&VectorFromFloat(Value=&FloatFromCurve(FloatCurve=DataInterface() {propertyValues {Curve {keys [
          {arriveTangent=0.1305, interpMode=RCIM_Cubic, leaveTangent=0.1305, tangentMode=RCTM_User, time=0, value=1}
          {arriveTangent=0.1729, interpMode=RCIM_Cubic, leaveTangent=0.173, tangentMode=RCTM_User, time=1, value=2.2}
         ]}}})))
   &ScaleColor(
    Scale Alpha=&FloatFromCurve(FloatCurve=DataInterface() {propertyValues {Curve {keys [<at 0 0>,<at 0.1397 1>,<at 1 0>]}}}, Scale Curve=$Particles.LightAlpha)
    ScaleA=true
    ScaleRGB=false
   )
   &DynamicMaterialParameters(
    Index 0 Param 1=&FloatFromCurve(FloatCurve=DataInterface() {propertyValues {Curve {keys [{arriveTangent=3.2486, interpMode=RCIM_Cubic, leaveTangent=3.2485, tangentMode=RCTM_User, time=0, value=0}, <at 1 2>]}}})
    Index 0 Param 4=&FloatFromCurve(FloatCurve=DataInterface() {propertyValues {Curve {keys [<at 0 5>,<at 1 6>]}}})
    Module.Param0WriteEnabled=true
    Module.Param1WriteEnabled=false
    Module.Param2WriteEnabled=false
    Module.Param3WriteEnabled=true
   )
  ]
  rendererClasses [&NiagaraMeshRendererProperties]
  renderers [{rendererClass=&NiagaraMeshRendererProperties, rendererIndex=0}]
 }
 NE_Decal {
  emitterUpdateScript [
   &EmitterState(
    Delay First Loop Only=true
    Inactive Response=::NewEnumerator1
    Life Cycle Mode=::NewEnumerator1
    Loop Delay=0.9
    Loop Duration=2.85
    UseLoopDelay=true
   )
   &SpawnRate(SpawnRate=20, Use Spawn Probability=false) {enabled=false}
   &SpawnBurst_Instantaneous(Spawn Count=1, Spawn Time=0.1, Use Spawn Probability=false)
  ]
  particleSpawnScript [
   &InitializeParticle(
    Lifetime=2
    Module.Write Color=true
    Module.Write Lifetime=true
    Module.Write Mass=true
    Module.Write Material Random=true
    Module.Write Position=false
    Module.Write Scale=false
    Module.Write SpriteRotation=false
    Module.Write SpriteSize=true
    Sprite Size=&Vector2DFromFloat(Value=40)
   )
   &CameraOffset(Camera Offset Amount=10)
   &DynamicMaterialParameters(
    Index 0 Param 1=1
    Index 0 Param 2=1
    Module.Param0WriteEnabled=true
    Module.Param1WriteEnabled=true
    Module.Param2WriteEnabled=false
    Module.Param3WriteEnabled=false
   )
  ]
  particleUpdateScript [
   &ParticleState()
   /SetVariables(Particles.LightAlpha:NiagaraFloat=&FloatFromCurve(FloatCurve=DataInterface() {propertyValues {Curve {keys [<at 0 0>,<at 0.1 1>,<at 0.45 1>,<at 0.75 0>]}}})) {isSetParameters=true, module=SetVariables_#1}
   &Color(
    Color=&ColorFromCurve(ColorCurve=DataInterface() {propertyValues {
       AlphaCurve {keys [<at -0.0046 0.2>]}
       BlueCurve {keys [<at -0.0017 0.16>]}
       GreenCurve {keys [<at -0.0017 0.03>]}
       RedCurve {keys [<at -0.0017 1>]}
      }})
    Scale Alpha=$Particles.LightAlpha
   )
   /NiagaraFN/Sprite/FN_SpriteAlignmentAndFacing(SpriteAlignment:Vector3f=<1 1 1>, SpriteFacing:Vector3f=<1 0 0>) {module=SpriteAlignmentAndFacing}
  ]
  rendererClasses [&NiagaraSpriteRendererProperties]
  renderers [{rendererClass=&NiagaraSpriteRendererProperties}]
 }
 NE_DecalCenter {
  emitterUpdateScript [
   &EmitterState(
    Delay First Loop Only=true
    Inactive Response=::NewEnumerator1
    Life Cycle Mode=::NewEnumerator1
    Loop Delay=0.9
    Loop Duration=2.85
    UseLoopDelay=true
   )
   &SpawnRate(SpawnRate=20, Use Spawn Probability=false) {enabled=false}
   &SpawnBurst_Instantaneous(Spawn Count=1, Spawn Time=0.1, Use Spawn Probability=false)
  ]
  particleSpawnScript [
   &InitializeParticle(
    Lifetime=2
    Module.Write Color=true
    Module.Write Lifetime=true
    Module.Write Mass=true
    Module.Write Material Random=true
    Module.Write Position=false
    Module.Write Scale=false
    Module.Write SpriteRotation=false
    Module.Write SpriteSize=true
    Sprite Size=&Vector2DFromFloat(Value=16)
   )
   &CameraOffset(Camera Offset Amount=10)
   &DynamicMaterialParameters(
    Index 0 Param 1=1
    Index 0 Param 2=1
    Module.Param0WriteEnabled=true
    Module.Param1WriteEnabled=true
    Module.Param2WriteEnabled=false
    Module.Param3WriteEnabled=false
   )
  ]
  particleUpdateScript [
   &ParticleState()
   /SetVariables(Particles.LightAlpha:NiagaraFloat=&FloatFromCurve(FloatCurve=DataInterface() {propertyValues {Curve {keys [<at 0 0>,<at 0.1 1>,<at 0.45 1>,<at 0.75 0>]}}})) {isSetParameters=true, module=SetVariables_#1}
   &Color(
    Color=&ColorFromCurve(ColorCurve=DataInterface() {propertyValues {
       AlphaCurve {keys [<at -0.0046 0.2>]}
       BlueCurve {keys [<at -0.0005 0.7262>,<at 0.1257 2.1785>,<at 0.5062 0.7262>]}
       GreenCurve {keys [<at -0.0005 0.6047>,<at 0.1257 1.8141>,<at 0.5062 0.6047>]}
       RedCurve {keys [<at -0.0005 1>,<at 0.1257 3>,<at 0.5062 1>]}
      }})
    Scale Alpha=$Particles.LightAlpha
   )
   &FN_SpriteAlignmentAndFacing(SpriteAlignment:Vector3f=<1 1 1>, SpriteFacing:Vector3f=<1 0 0>) {module=SpriteAlignmentAndFacing}
  ]
  rendererClasses [&NiagaraSpriteRendererProperties]
  renderers [{rendererClass=&NiagaraSpriteRendererProperties}]
 }
}
summary {
 systemName=NS_DeliSandwich_Sonar_Backpack1
 emitters [
  {emitterName=DeliSandwish_VFX, rendererClasses [&NiagaraMeshRendererProperties]}
  {emitterName=NE_Decal, rendererClasses [&NiagaraSpriteRendererProperties]}
  {emitterName=NE_DecalCenter, rendererClasses [&NiagaraSpriteRendererProperties]}
 ]
 userVariables []
}
